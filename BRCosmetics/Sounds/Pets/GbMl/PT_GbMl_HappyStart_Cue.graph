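# Sound cue for the GbMl pet's "HappyStart" vocal, played in the creatures/pets sound class with close-range pet attenuation.
# A quality-level node at the root sends its first two slots to a random node that picks among eight equally weighted Happy wave variations.
# Its last two slots play a single fixed variation, Happy_10, skipping the random pick at those quality levels.
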SoundCue /BRCosmetics/Sounds/Pets/GbMl/PT_GbMl_HappyStart_Cue
root=QualityLevel_0
nodes [
 {id=QualityLevel_0, children [Random_0, Random_0, WavePlayer_12, WavePlayer_12]}
 {
  id=Random_0
  children [WavePlayer_12, WavePlayer_13, WavePlayer_14, WavePlayer_15, WavePlayer_16, WavePlayer_17, WavePlayer_18, WavePlayer_19]
  properties {weights [1,1,1,1,1,1,1,1]}
 }
 {id=WavePlayer_12, properties {soundWaveAssetPtr=&/PT_GbMl_Happy_10}}
 {id=WavePlayer_13, properties {soundWaveAssetPtr=&/PT_GbMl_Happy_11}}
 {id=WavePlayer_14, properties {soundWaveAssetPtr=&/PT_GbMl_Happy_12}}
 {id=WavePlayer_15, properties {soundWaveAssetPtr=&/PT_GbMl_Happy_13}}
 {id=WavePlayer_16, properties {soundWaveAssetPtr=&/PT_GbMl_Happy_04}}
 {id=WavePlayer_17, properties {soundWaveAssetPtr=&/PT_GbMl_Happy_05}}
 {id=WavePlayer_18, properties {soundWaveAssetPtr=&/PT_GbMl_Happy_06}}
 {id=WavePlayer_19, properties {soundWaveAssetPtr=&/PT_GbMl_Happy_09}}
]
properties {
 attenuationSettings=&Fort_Pets_Close_Athena_Att
 maxDistance=1800.0001
 soundClassObject=&Fort_Creatures_Pets
 volumeMultiplier=1.12
}
